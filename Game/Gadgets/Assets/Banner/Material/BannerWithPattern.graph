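# Two-sided masked material for banners and flags, usable on skeletal meshes, that prints a tintable icon over a rotatable two-color gradient fabric.
# The icon's green channel, scaled about center and optionally masked by a banner shape, also blends roughness and metallic between fabric and icon.
# A Flag Flutter switch selects panning swirl-noise normals and displacement weighted by red vertex color, or the shared Cloth function's offset and normals over a tarp normal map.
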Material /Game/Gadgets/Assets/Banner/Material/BannerWithPattern
graph {
 properties {
  bUsedWithMeshDeformer=true
  bUsedWithSkeletalMesh=true
  blendMode=BLEND_Masked
  pixelDepthOffsetMode=PDOM_Legacy
  previewMesh=/SaveTheWorld/Environments/Homebase/Meshes/SM_HB_Center_Flag_Repositioned_noLOD
  twoSided=true
  usedShadingModels=Default Lit
 }
 nodes [
  Constant3Vector() {constant=<rgba 30 15 0 1>}
  Constant3Vector() {constant=<rgba -1 -1 1 0>}
  Constant3Vector() {constant=<rgba -0.5 -0.5 1 1>}
  n1 = TextureCoordinate()
  n2 = VectorParameter() {parameterName=Barnner Icon Translation, defaultValue=<rgba 1 1 0 0>}
  n3 = AppendVector(A=@n2.R, B=@n2.R)
  n4 = MaterialFunctionCall(
   UVs=Add(A=Multiply(A=@n1, B=@n3), B=AppendVector(A=@n2.B, B=@n2.B))
   Texture Scale=AppendVector(A=ScalarParameter() {defaultValue=1, parameterName=IconScaleU}, B=ScalarParameter() {defaultValue=1, parameterName=IconScaleV})
  ) {materialFunction=&ScaleUVsByCenter}
  n5 = TextureSampleParameter2D(UVs=@n4.UVs) {
   parameterName=BannerIcon
   samplerSource=SSM_Clamp_WorldGroupSettings
   texture=/Game/UI/Foundation/Textures/Banner/Standard/T-Banners-Icons-Rays-L
  }
  n6 = Multiply(A=@n5.G) {constB=2}
  n7 = TextureSample(UVs=@n4.UVs) {texture=/Game/UI/Foundation/Textures/Banner/Shapes/banner_shape_standard}
  n8 = Clamp(None=StaticSwitchParameter(True=Multiply(A=@n6, B=@n7.R), False=@n6) {defaultValue=true, parameterName=BannerIconShapeMask})
  n9 = StaticSwitchParameter(True=Multiply(A=@n8, B=TextureSample(UVs=@n4.UVs) {texture=/Game/Environments/Hexmap/Outpost/T-Banners-Icons-Mask}.R), False=@n8) {parameterName=Flag Flutter}
  LinearInterpolate(
   A=ScalarParameter() {defaultValue=0.35, parameterName=SpecRoughnessMin}
   B=ScalarParameter() {defaultValue=0.5, parameterName=SpecRoughnessMax}
   Alpha=@n9
  )
  n10 = TextureCoordinate()
  n11 = ScalarParameter() {group=Linear Gradient Settings, parameterName=ColorRotation}
  n12 = Subtract(A=ComponentMask(None=Rotator(Coordinate=@n10, Time=@n11)) {g=true}, B=ScalarParameter() {defaultValue=-0.4, parameterName=ColorOffset})
  n13 = ScalarParameter() {group=Linear Gradient Settings, parameterName=RGBAMin}
  n14 = ScalarParameter() {defaultValue=1, group=Linear Gradient Settings, parameterName=RGBAMax}
  LinearInterpolate(
   A=LinearInterpolate(
    A=VectorParameter() {parameterName=RGBA0, defaultValue=<rgba 1 0.2719 0 1>}.RGB
    B=VectorParameter() {parameterName=RGBA1, defaultValue=<rgba 0.28 0.0595 0 1>}.RGB
    Alpha=MaterialFunctionCall(Alpha=@n12, Min=@n13, Max=@n14) {materialFunction=&SmoothStep}.Result
   )
   B=VectorParameter() {parameterName=BannerIconColor, defaultValue=<rgba 1 1 1 1>}.RGB
   Alpha=@n9
  )
  LinearInterpolate(A=ScalarParameter() {defaultValue=0.44, parameterName=Metallic Fabric}, B=ScalarParameter() {parameterName=Metallic Icon}, Alpha=@n9)
  ScalarParameter() {defaultValue=0.12, parameterName=Specular}
  n15 = TextureCoordinate() {vTiling=1.5}
  n16 = ScalarParameter() {defaultValue=0.3, parameterName=Scale}
  n17 = Time()
  n18 = ScalarParameter() {defaultValue=2, parameterName=Speed}
  n19 = TextureSample(UVs=Panner(Coordinate=Multiply(A=@n15, B=@n16), Time=Multiply(A=@n17, B=@n18)) {speedX=0.1, speedY=0.025}) {samplerType=SAMPLERTYPE_Normal, texture=/Game/Global/Textures/Default/Noise/T_SwirlNoise_N}
  n20 = Multiply(A=@n19.RGB, B=Constant3Vector() {constant=<rgba -1 -1 1 0>})
  n21 = VectorParameter() {parameterName=WPOPower, defaultValue=<rgba 30 10 5 1>}
  n22 = VertexColor()
  n23 = Constant() {r=3}
  n24 = MaterialFunctionCall!(
   UVs
   Pan Speed=ScalarParameter() {defaultValue=-1, group=WPO_ClothAnimation, parameterName=WPO_ClothPanSpeed}
   WPO_ClothOffsetDirection
   WPO_ClothNormalIntensity
   WPO_ClothOffsetIntensity
   "Motion Mask (Defaults to the Red Vertex Color)"
   WPO_ClothWPOWindTextureOffset
   Pan Direction
  ) {materialFunction=/Game/Packages/Fortress_SharedMaterials/Base_Material_Functions/Cloth}
  StaticSwitchParameter(
   True=LinearInterpolate(A=Constant3Vector() {constant=<rgba 0 0 1 0>}, B=Multiply(A=@n20, B=@n21.RGB), Alpha=Power(Base=@n22.R, Exp=@n23))
   False=Divide(A=@n24.Additive WPO)
  ) {parameterName=Flag Flutter}
  n25 = Constant3Vector() {constant=<rgba 0 0 1 0>}
  n26 = TextureCoordinate() {uTiling=0.5}
  n27 = ScalarParameter() {defaultValue=1.5, parameterName=ScaleB}
  n28 = ScalarParameter() {defaultValue=1.25, parameterName=SpeedB}
  n29 = TextureSample(UVs=Panner(Coordinate=Multiply(A=@n26, B=@n27), Time=Multiply(A=@n17, B=@n28)) {speedX=0.1, speedY=-0.05}) {samplerType=SAMPLERTYPE_Normal, texture=&T_SwirlNoise_N}
  n30 = VectorParameter() {parameterName=SmallNormalsPower, defaultValue=<rgba 0.5 0.5 1 1>}
  n31 = Multiply(
   A=MaterialFunctionCall(BaseNormal=@n19.RGB, AdditionalNormal=Multiply(A=@n29.RGB, B=@n30.RGB) {constB=-1}) {materialFunction=&BlendAngleCorrectedNormals}.Result
   B=Constant3Vector() {constant=<rgba -1 -1 1 0>}
  )
  n32 = VectorParameter() {parameterName=Normal Translation, defaultValue=<rgba 1 1 0 0>}
  n33 = Add(A=Multiply(A=TextureCoordinate(), B=AppendVector(A=@n32.R, B=@n32.R)), B=AppendVector(A=@n32.B, B=@n32.B))
  StaticSwitchParameter(
   True=Multiply(A=VectorParameter() {parameterName=MasterNormalPower, defaultValue=<rgba 1 1 0.5 1>}.RGB, B=LinearInterpolate(A=@n25, B=@n31, Alpha=@n22.R))
   False=MaterialFunctionCall(
    BaseNormal=TextureSampleParameter2D(UVs=@n33) {parameterName=Normal, samplerType=SAMPLERTYPE_Normal, texture=/Game/Environments/Sets/Junkyard/Textures/T_Tarp_N}.RGB
    AdditionalNormal=@n24.Normal
   ) {materialFunction=&BlendAngleCorrectedNormals}.Result
  ) {parameterName=Flag Flutter}
  TextureCoordinate() {coordinateIndex=2}
 ]
}
