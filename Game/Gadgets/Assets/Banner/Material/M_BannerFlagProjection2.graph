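# Unlit translucent particle sprite material for the Banner gadget, apparently a glowing blue flag projection effect.
# A constant blue color is multiplied up for brightness and passed through the MF_FX_EmissiveAdjustments function.
# Particle alpha times the green channel of a sphere mask texture is clamped, giving a soft mask that particles can fade.
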Material /Game/Gadgets/Assets/Banner/Material/M_BannerFlagProjection2
graph {
 properties {
  bUsedWithParticleSprites=true
  blendMode=BLEND_Translucent
  pixelDepthOffsetMode=PDOM_Legacy
  shadingModel=MSM_Unlit
  usedShadingModels=Unlit
 }
 nodes [
  Clamp(None=Multiply(A=ParticleColor().A, B=TextureSample() {texture=/Game/Effects/Fort_Effects/Textures/Gameplay/T_Sphere_Mask}.G) {constB=0})
  MaterialFunctionCall(Emissive=Multiply(A=Constant3Vector() {constant=<rgba 0.0113 0.0386 0.94 0>}) {constB=40}) {materialFunction=&MF_FX_EmissiveAdjustments}
 ]
}
