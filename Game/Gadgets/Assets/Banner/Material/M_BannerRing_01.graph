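# M_BannerRing_01 is an unlit translucent particle material for a banner gadget's ring effect, with animated ripples and tinted glow.
# Two panning water-ripple normals combine to distort an inverted Fresnel and drive world position offset along vertex normals.
# That Fresnel plus vertex red, minus a dynamic Erode value, shapes depth-faded opacity and a Ceil-stepped color blend.
# A second Fresnel and Color LERP Bias grade the colors, multiplied by particle color.
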Material /Game/Gadgets/Assets/Banner/Material/M_BannerRing_01
graph {
 properties {
  bEnableResponsiveAA=true
  bUsedWithMeshParticles=true
  blendMode=BLEND_Translucent
  opacityMaskClipValue=0.1
  pixelDepthOffsetMode=PDOM_Legacy
  shadingModel=MSM_Unlit
  usedShadingModels=Unlit
 }
 nodes [
  n1 = ParticleColor()
  n2 = ScalarParameter() {defaultValue=0.5, parameterName=Base Opacity}
  n3 = ScalarParameter() {defaultValue=1, parameterName=Tile U1}
  n4 = ScalarParameter() {defaultValue=0.228, parameterName=Tile V1}
  n5 = Panner(Coordinate=Multiply(A=AppendVector(A=@n3, B=@n4), B=TextureCoordinate())) {speedX=0.1225, speedY=-0.2109}
  n6 = ScalarParameter() {defaultValue=3, parameterName=Tile U2}
  n7 = ScalarParameter() {defaultValue=0.341, parameterName=Tile V2}
  n8 = Panner(Coordinate=Multiply(A=AppendVector(A=@n6, B=@n7), B=TextureCoordinate())) {speedX=-0.283, speedY=-0.178}
  n9 = Normalize(VectorInput=Add(
    A=TextureSample(UVs=@n5) {samplerType=SAMPLERTYPE_Normal, texture=/Game/Effects/Fort_Effects/Gameplay/Water/Textures/T_WaterRipples_01}.RGB
    B=TextureSample(UVs=@n8) {samplerType=SAMPLERTYPE_Normal, texture=&T_WaterRipples_01}.RGB
   ))
  n10 = Add(
   A=MaterialFunctionCall(
    Brightness=ScalarParameter() {defaultValue=1, parameterName=Fresnel Brightness Edge}
    Exponent=ScalarParameter() {defaultValue=4, parameterName=Fresnel Exponent Edge}
    Vertex / Pixel Normal=Transform(None=@n9)
   ) {materialFunction=&MF_Fresnel_FX}.Fresnel Inverted
   B=VertexColor().R
  )
  n11 = Add(
   A=Subtract(A=Clamp(None=@n10), B=DynamicParameter() {defaultValue=<rgba 0 0 1 1>, paramNames [Erode, Color LERP Bias, Param3, Param4]}.Erode)
   B=ScalarParameter() {defaultValue=-0.2, parameterName=Bias}
  )
  DepthFade(
   Opacity=Multiply(A=@n1.A, B=LinearInterpolate(A=@n2, Alpha=@n11))
   FadeDistance=ScalarParameter() {defaultValue=30, parameterName=Intersect Distance}
  )
  n12 = VectorParameter() {parameterName=Color A, defaultValue=<rgba 0.0493 0.1244 1 1>}
  n13 = Add(
   A=VertexColor().R
   B=MaterialFunctionCall(Exponent=ScalarParameter() {defaultValue=5, parameterName="Fresnel Exponent "}) {materialFunction=&MF_Fresnel_FX}.Fresnel Inverted
  )
  n14 = DynamicParameter() {defaultValue=<rgba 0 0 1 1>, paramNames [Erode, Color LERP Bias, Param3, Param4]}
  n15 = LinearInterpolate(
   A=VectorParameter() {parameterName=Color B1, defaultValue=<rgba 0.01 0.0142 0.115 1>}.RGB
   B=VectorParameter() {parameterName=Color B2, defaultValue=<rgba 0.6832 1.41 30 1>}.RGB
   Alpha=Clamp(None=Add(A=@n13, B=@n14.Color LERP Bias))
  )
  n16 = Add(A=@n14.Color LERP Bias, B=Ceil(None=@n11))
  MaterialFunctionCall(Emissive=Multiply(A=LinearInterpolate(A=@n12.RGB, B=@n15, Alpha=@n16), B=@n1.RGB)) {materialFunction=&MF_FX_EmissiveAdjustments}
  Multiply(A=Multiply(A=@n9, B=VertexNormalWS()), B=VectorParameter() {parameterName=WPO, defaultValue=<rgba 70 70 20 0>}.RGB)
 ]
}
